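# Sound cue for the Singing Bass gadget, playing one of three looping Singing Bass sound waves.
# A Random node with equal weights chooses between the three looping wave players.
# The cue plays at reduced volume through the Fort_Music_Quest_PSM sound class and overrides attenuation with its own falloff, stereo spread and non-spatialized radius for positional playback.
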SoundCue /Game/Sounds/Fort_Gadgets/SingingBass/SingingBass_Cue
root=Random_2
nodes [
 {id=Random_2, children [WavePlayer_0,WavePlayer_1,WavePlayer_2], properties {weights [1,1,1]}}
 {id=WavePlayer_0, properties {bLooping=true, soundWaveAssetPtr=&/SingingBass_01}}
 {id=WavePlayer_1, properties {bLooping=true, soundWaveAssetPtr=&/SingingBass_03}}
 {id=WavePlayer_2, properties {bLooping=true, soundWaveAssetPtr=&/SingingBass_02}}
]
properties {
 bOverrideAttenuation=true
 maxDistance=2600
 soundClassObject=/Game/Sounds/FortSoundClassesAndModes/Fort_Music_Quest_PSM
 volumeMultiplier=0.65
 attenuationOverrides {falloffDistance=1800, nonSpatializedRadiusStart=200, stereoSpread=100, attenuationShapeExtents {x=800}}
}
